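; Looping gameplay cue that drives the Toy Fox boss's stagger immunity visual on its budgeted skeletal mesh.
; Application and removal each call the parent, cast the target to the Toy Fox boss, and call an ON or OFF helper.
; The helpers write custom primitive data: a state flag in slot 35 (0 for on, 1 for off) and game time in slot 34, giving the material a transition timestamp.
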(event EventOnApplication(EitherVFX-preferthisone)
  (params (MyTarget :type Actor%20Object%20Reference) (Parameters :type Gameplay%20Cue%20Parameters%20Structure%20%28by%20ref%29) (ParticleComponents :type Array%20of%20FXSystem%20Component%20Object%20References) (AudioComponents :type Array%20of%20Audio%20Component%20Object%20References) (BurstCameraShakeInstance :type Camera%20Shake%20Base%20Object%20Reference) (BurstDecalInstance :type Decal%20Actor%20Object%20Reference))
  (|CallParentFunction MyTarget Parameters ParticleComponents AudioComponents BurstCameraShakeInstance BurstDecalInstance)
  (bind _asbga_tidal_crane_boss_toy_fox (Utilities|Casting|CastToBGA_TidalCrane_Boss_ToyFox MyTarget))
  (CallFunction|MStaggerImmunityON (Variables|/TidalCrane_Swarm_Boss/Gameplay/BGA_TidalCrane_Boss.BGA_TidalCrane_Boss_C|GetSkeletalMeshComponentBudgeted _asbga_tidal_crane_boss_toy_fox)))

(event EventOnRemoval(EitherVFX-preferthisversion)
  (params (MyTarget :type Actor%20Object%20Reference) (Parameters :type Gameplay%20Cue%20Parameters%20Structure%20%28by%20ref%29) (ParticleComponents :type Array%20of%20FXSystem%20Component%20Object%20References) (AudioComponents :type Array%20of%20Audio%20Component%20Object%20References) (BurstCameraShakeInstance :type Camera%20Shake%20Base%20Object%20Reference) (BurstDecalInstance :type Decal%20Actor%20Object%20Reference))
  (|CallParentFunction MyTarget Parameters ParticleComponents AudioComponents BurstCameraShakeInstance BurstDecalInstance)
  (bind _asbga_tidal_crane_boss_toy_fox (Utilities|Casting|CastToBGA_TidalCrane_Boss_ToyFox MyTarget))
  (MaterialAnim|MStaggerImmunityOFF (Variables|/TidalCrane_Swarm_Boss/Gameplay/BGA_TidalCrane_Boss.BGA_TidalCrane_Boss_C|GetSkeletalMeshComponentBudgeted _asbga_tidal_crane_boss_toy_fox)))

(fn M_StaggerImmunityOFF
  (params (MeshTarget :type Primitive%20Component%20Object%20Reference))
  (Rendering|Material|SetCustomPrimitiveDataFloat MeshTarget 35 1.0)
  (Rendering|Material|SetCustomPrimitiveDataFloat MeshTarget 34 (Utilities|Time|GetGameTimeinSeconds 0)))

(fn M_StaggerImmunityON
  (params (MeshTarget :type Primitive%20Component%20Object%20Reference))
  (Rendering|Material|SetCustomPrimitiveDataFloat MeshTarget 35 0.0)
  (Rendering|Material|SetCustomPrimitiveDataFloat MeshTarget 34 (Utilities|Time|GetGameTimeinSeconds 0)))

(fn ConstructionScript
  (|CallParentFunction))
