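; Burst gameplay cue for the Toy Fox boss death, which plays the de-rez dissolve effect.
; The OnBurst override runs the parent burst first, then casts the target to the Toy Fox boss blueprint and calls its DeathDissolve.
; It then moves the first spawned particle component to a fixed, hard-coded relative offset forward and up.
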(fn OnBurst(EitherVFX-preferthisversion)
  (params (MyTarget :type Actor%20Object%20Reference) (Parameters :type Gameplay%20Cue%20Parameters%20Structure%20%28by%20ref%29) (ParticleComponents :type Array%20of%20FXSystem%20Component%20Object%20References) (AudioComponents :type Array%20of%20Audio%20Component%20Object%20References) (BurstCameraShakeInstance :type Camera%20Shake%20Base%20Object%20Reference) (BurstDecalInstance :type Decal%20Actor%20Object%20Reference))
  ;[0 0.15 0.06086 0.5] OnBurst: Play the death dissolve FX.
  (|CallParentFunction MyTarget Parameters ParticleComponents AudioComponents BurstCameraShakeInstance BurstDecalInstance)
  (Utilities|FlowControl|Sequence
    (:then_0
      (bind _asbga_tidal_crane_boss_toy_fox (Utilities|Casting|CastToBGA_TidalCrane_Boss_ToyFox MyTarget))
      (Class|BGATidalCraneBoss|DeathDissolve _asbga_tidal_crane_boss_toy_fox))
    (:then_1
      (Transformation|SetRelativeLocation (Utilities|Array|Get(acopy) ParticleComponents) "1200.000000,0.000000,500.000000" false false))
    (:then_2
      (return))))
