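; Burst gameplay cue for the giant ToyFox boss melee impact, with a visual variant for the headbutt attack.
; OnBurst runs the parent burst, then treats a gameplay effect level of 1 in the cue parameters as a headbutt.
; For a headbutt, it deactivates the first particle component, swaps in a TidalCrane Swarm boss projectile impact Niagara system and reactivates it with a reset.
; Other levels keep the parent's effects unchanged.
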(fn OnBurst(EitherVFX-preferthisversion)
  (params (MyTarget :type Actor%20Object%20Reference) (Parameters :type Gameplay%20Cue%20Parameters%20Structure%20%28by%20ref%29) (ParticleComponents :type Array%20of%20FXSystem%20Component%20Object%20References) (AudioComponents :type Array%20of%20Audio%20Component%20Object%20References) (BurstCameraShakeInstance :type Camera%20Shake%20Base%20Object%20Reference) (BurstDecalInstance :type Decal%20Actor%20Object%20Reference))
  (bind (_normalizedmagnitude _rawmagnitude _effectcontext _matchedtagname _originaltag _aggregatedsourcetags _aggregatedtargettags _location _normal _instigator _effectcauser _sourceobject _physicalmaterial _gameplayeffectlevel) (Ability|GameplayCue|BreakGameplayCueParameters Parameters))
  (bind _output (Utilities|Array|Get(acopy) ParticleComponents))
  (bind _asniagara_particle_system_component (Utilities|Casting|CastToNiagaraComponent _output))
  (|CallParentFunction MyTarget Parameters ParticleComponents AudioComponents BurstCameraShakeInstance BurstDecalInstance)
  (bind _output_get (LocalVariables|SetbHeadbuttAttack (== _gameplayeffectlevel 1)))
  (if _output_get
    (Components|Activation|Deactivate _output)
    (Niagara|SetNiagaraSystemAsset _asniagara_particle_system_component "/TidalCrane_Swarm_Boss/Assets/FX/EventVariant/Melee/NS_TidalCrane_Event_Projectile_Impact.NS_TidalCrane_Event_Projectile_Impact" false)
    (Components|Activation|SetActive _asniagara_particle_system_component true true))
  (return))
