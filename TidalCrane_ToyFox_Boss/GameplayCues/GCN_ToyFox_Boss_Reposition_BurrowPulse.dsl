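; Burst gameplay cue for the ToyFox boss's reposition burrow pulse, adjusting its particle effect for water hits and surface orientation.
; After the parent burst runs, it casts the hit component's owner to a water body; on success it records the impact location and sets bIsInWater on the first particle component.
; It then rotates that component so its Z axis opposes the cue normal, deriving X from a cross product with world up.
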(fn OnBurst(EitherVFX-preferthisversion)
  (params (MyTarget :type Actor%20Object%20Reference) (Parameters :type Gameplay%20Cue%20Parameters%20Structure%20%28by%20ref%29) (ParticleComponents :type Array%20of%20FXSystem%20Component%20Object%20References) (AudioComponents :type Array%20of%20Audio%20Component%20Object%20References) (BurstCameraShakeInstance :type Camera%20Shake%20Base%20Object%20Reference) (BurstDecalInstance :type Decal%20Actor%20Object%20Reference))
  (bind (_normalizedmagnitude _rawmagnitude _effectcontext) (Ability|GameplayCue|BreakGameplayCueParameters Parameters))
  (bind _returnvalue (Ability|EffectContext|GetHitResult _effectcontext))
  (bind (_bblockinghit _binitialoverlap _time _distance _location _impactpoint _normal _impactnormal _physmat _hitactor _hitcomponent) (Collision|BreakHitResult _returnvalue))
  (bind (_normalizedmagnitude_2 _rawmagnitude_2 _effectcontext_2 _matchedtagname _originaltag _aggregatedsourcetags _aggregatedtargettags _location_2 _normal_2) (Ability|GameplayCue|BreakGameplayCueParameters Parameters))
  (bind _output (Utilities|Array|Get(acopy) ParticleComponents))
  (|CallParentFunction MyTarget Parameters ParticleComponents AudioComponents BurstCameraShakeInstance BurstDecalInstance)
  (Utilities|FlowControl|Sequence
    (:then_0
      ;[0 0.15 0.04628 0.5] Assess if impact is on water.
      (bind _asfort_water_body_actor (Utilities|Casting|CastToFortWaterBodyActor (Components|GetOwner _hitcomponent)))
      (bind (_outwatersurfaceinfo _returnvalue_2) (WaterBody|GetWaterSurfaceInfoStruct _asfort_water_body_actor _impactpoint false false false))
      (LocalVariables|SetWaterImpactLocation (select _returnvalue_2 (.WaterSurfaceLocation (Utilities|Struct|BreakWaterSurfaceInfo _outwatersurfaceinfo)) _impactpoint))
      (LocalVariables|SetbWaterImpact true)
      (Effects|Components|ParticleSystem|SetBoolParameter _output "bIsInWater" true))
    (:then_1
      (Transformation|SetWorldRotation _output (Math|Rotator|MakeRotfromZX (* _normal_2 "-1.000000,-1.000000,-1.000000") (Math|Vector|Normalize (Math|Vector|CrossProduct (* _normal_2 "-1.000000,-1.000000,-1.000000") "0.000000,0.000000,1.000000"))) false false))
    (:then_2
      (return))))
